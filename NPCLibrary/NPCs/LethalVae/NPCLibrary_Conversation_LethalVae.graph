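# Conversation database for the Lethal Vae NPC, giving her greeting dialogue about monster hunting and pursuing Kado Thorne.
# A single entry point leads to one speech node with a default line and contextual overrides.
# Overrides fire on a first-ever conversation or when the player wears specific character outfits, ranked by curve-driven priorities.
# The speech then links to the shared NPC Library quests and basic services conversations.
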ConversationDatabase /NPCLibrary/NPCs/LethalVae/NPCLibrary_Conversation_LethalVae
entry_points [{tag=Conversation.Entry.LethalVae, destinations [#1]}]
nodes [
 ConversationEntryPointNode_7 {
  id=#1
  class=ConversationEntryPointNode
  outputs [#2]
  properties {entryTag {tagName=Conversation.Entry.LethalVae}}
 }
 Speech_C_4 {
  id=#2
  class=Speech_C
  outputs [#3, #4]
  properties {
   nodeName=""
   generalConfig {
    defaultMessage=Kado Thorne isn't getting away from me this time.
    contextualMessages [
     {
      message=Shh. I'm monster hunting.
      contextRequirements [{requirement {class=FortControllerRequirement_IsFirstEverConversationWithNPC}}]
      priority {curve {rowName=Tandem.Speech.Priority.FirstEverInteraction}}
     }
     {
      message="Brave of you to talk to me in the open. You really think you're untouchable, don't you?"
      contextRequirements [{requirement {class=FortControllerRequirement_HasCID, properties {allowedCIds [/BRCosmetics/Athena/Items/Cosmetics/Characters/Character_LazarusLens]}}}]
      priority {curve {rowName=Tandem.Speech.Priority.OutfitReaction}}
     }
     {
      message=I hear you've taken on a protégé. I hope you learn as much teaching her as I learned teaching you.
      contextRequirements [{requirement {
         class=FortControllerRequirement_HasCID
         properties {allowedCIds [/BRCosmetics/Athena/Items/Cosmetics/Characters/CID_869_Athena_Commando_F_Cavalry]}
        }}]
      priority {curve {rowName=Tandem.Speech.Priority.OutfitReaction}}
     }
     {
      message="You'll learn a lot from Victoria Saint, but you'll have to give up everything to become a true hunter."
      contextRequirements [{requirement {
         class=FortControllerRequirement_HasCID
         properties {allowedCIds [/BRCosmetics/Athena/Items/Cosmetics/Characters/Character_Silencer, /BRCosmetics/Athena/Items/Cosmetics/Characters/Character_ProdigyFire]}
        }}]
      priority {curve {rowName=Tandem.Speech.Priority.OutfitReaction}}
     }
     {
      message="Your friend Joni will have to choose, someday, between her friends and the true hunt. Don't hold her back."
      contextRequirements [{requirement {
         class=FortControllerRequirement_HasCID
         properties {allowedCIds [&Character_SharpFang, /BRCosmetics/Athena/Items/Cosmetics/Characters/Character_Sparrow]}
        }}]
      priority {curve {rowName=Tandem.Speech.Priority.OutfitReaction}}
     }
     {
      message=Did you get what you needed from Kado Thorne? Am I clear to make my move without getting in your way?
      contextRequirements [{requirement {class=FortControllerRequirement_HasCID, properties {allowedCIds [&Character_InnovatorSand]}}}]
      priority {curve {rowName=Tandem.Speech.Priority.OutfitReaction}}
     }
    ]
   }
  }
 }
 ConversationLinkNode_0 {id=#4, class=ConversationLinkNode, properties {remoteEntryTag {tagName=Conversation.Entry.BasicServices}}}
 ConversationLinkNode_1 {id=#3, class=ConversationLinkNode, properties {remoteEntryTag {tagName=Conversation.Entry.NPCLibrary.Quests}}}
]
